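# M_Grapefruit_Dots is a masked, lit material for Niagara mesh particles in the Grapefruit emote effect.
# Particle color drives two branches: one scales it by a Diffuse Mult parameter for the diffuse contribution.
# The other blends from a constant toward that color by Add Emissive, then feeds the MF_FX_EmissiveAdjustments function.
# An MSR vector parameter is also defined.
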Material /BRCosmetics/Effects/Fort_Effects/Emotes/Grapefruit/M_Grapefruit_Dots
graph {
 properties {bUsedWithNiagaraMeshParticles=true, blendMode=BLEND_Masked, pixelDepthOffsetMode=PDOM_Legacy, usedShadingModels=Default Lit}
 nodes [
  n1 = ParticleColor()
  MaterialFunctionCall(Emissive=LinearInterpolate(A=Constant(), B=@n1.RGB, Alpha=ScalarParameter() {parameterName=Add Emissive})) {materialFunction=&MF_FX_EmissiveAdjustments}
  Multiply(A=@n1.RGB, B=ScalarParameter() {defaultValue=1, desc=Diffuse Contribution, parameterName=Diffuse Mult, sliderMax=1})
  VectorParameter() {parameterName=MSR, defaultValue=<rgba 0 0 0 1>}
 ]
}
