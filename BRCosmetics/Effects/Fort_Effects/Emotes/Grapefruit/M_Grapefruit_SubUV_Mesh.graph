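# Masked lit material for Niagara mesh and sprite particles in the Grapefruit emote, showing SubUV frames from the Grapefruit backpack texture.
# Dynamic parameters supply frame index and grid size, with Use Variant blending to a second set.
# Vertex-color blue, marking sides, blends toward particle color when the variant index is low.
# Alpha multiplies frame and particle alpha by a local-space XZ noise wipe driven by System Age, feeding DitherTemporalAA.
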Material /BRCosmetics/Effects/Fort_Effects/Emotes/Grapefruit/M_Grapefruit_SubUV_Mesh
graph {
 properties {
  bEnableResponsiveAA=true
  bOutputTranslucentVelocity=true
  bUsedWithNiagaraMeshParticles=true
  bUsedWithNiagaraSprites=true
  blendMode=BLEND_Masked
  opacityMaskClipValue=0.01
  pixelDepthOffsetMode=PDOM_Legacy
  usedShadingModels=Default Lit
 }
 nodes [
  Constant() {r=1}
  Constant() {r=1}
  n1 = ParticleColor()
  n2 = Saturate(None=DynamicParameter() {parameterIndex=1, paramNames [Variant_SubUVIndex, ParticleUV_x, ParticleUV_y, Variant_NumSprites]}.Variant_SubUVIndex)
  n3 = TextureObjectParameter() {parameterName=TextureObject Parm, texture=/BRCosmetics/Accessories/FORT_Backpacks/Backpack_F_MED_Grapefruit/VFX/T_Grapefruit_Backpack_D}
  n4 = DynamicParameter() {paramNames [SubUVIndex,ParticleUV_x,ParticleUV_y,NumSprites]}
  n5 = DynamicParameter() {desc=state, parameterIndex=1, paramNames [Variant_SubUVIndex, ParticleUV_x, ParticleUV_y, Variant_NumSprites]}
  n6 = ScalarParameter() {parameterName=Use Variant, sliderMax=1}
  n7 = LinearInterpolate(A=@n4.NumSprites, B=@n5.Variant_NumSprites, Alpha=@n6)
  n8 = LinearInterpolate(A=@n4.SubUVIndex, B=@n5.Variant_SubUVIndex, Alpha=@n6)
  n9 = MaterialFunctionCall!(Texture=@n3, UVs=TextureCoordinate(), SubImages=AppendVector(A=@n7, B=@n7), Frame=@n8) {materialFunction=/Engine/Functions/Engine_MaterialFunctions03/Texturing/SubUV_Function_MipDerivative}
  LinearInterpolate(A=@n9.RGB, B=@n1.RGB, Alpha=Multiply(A=OneMinus(None=@n2), B=VertexColor() {desc=blue for sides or red for front}.B))
  MaterialFunctionCall() {materialFunction=&DitherTemporalAA_Fortnite}
  n10 = Multiply(
   A=MaterialFunctionCall() {materialFunction=&LocalPosition}.Instance Local Position
   B=ScalarParameter() {defaultValue=0.1, parameterName=Wipe Freq, sliderMax=1}
  )
  n11 = TextureObjectParameter() {parameterName=WipeOn Texture, samplerType=SAMPLERTYPE_Masks, texture=/Game/Environments/Apollo/Sets/CubeTownV2/Materials/Parent/CubeTown_1_H}
  n12 = Subtract(A=TextureSample(UVs=ComponentMask(None=@n10) {b=true, r=true}, Tex=@n11).RGB, B=Constant3Vector() {constant=<rgba 1 1 1 0>})
  n13 = Power(Base=InverseLinearInterpolate(
    B=ScalarParameter() {defaultValue=0.5, parameterName=Wipe Speed, sliderMax=15}
    Value=DynamicParameter() {parameterIndex=3, paramNames [System Age, "", "", ""]}.System Age
   ) {bClampResult=true, constA=0.1}) {constExponent=0.5}
  n14 = Multiply(A=Multiply(A=@n9.Alpha, B=@n1.A), B=Saturate(None=Add(A=@n12, B=@n13)))
  MaterialFunctionCall(Alpha Threshold=@n14) {materialFunction=&DitherTemporalAA}
  Saturate(None=Power(Base=@n14) {constExponent=0.01})
  Saturate(None=@n14)
  n15 = TextureObjectParameter() {parameterName=TextureObject Parm, texture=&T_Grapefruit_Backpack_D}
  n16 = Fmod(A=Floor(None=@n8), B=@n7)
  n17 = MaterialFunctionCall(Float2=TextureCoordinate()) {materialFunction=&BreakOutFloat2Components}
  n18 = Floor(None=Divide(A=@n8, B=@n7))
  n19 = AppendVector(
   A=Add(A=Divide(A=@n16, B=@n7) {desc=fmod/N}, B=Divide(A=@n17.R, B=@n7) {desc=u in 1/N})
   B=Add(A=Divide(A=@n18, B=@n7) {desc=div/N}, B=Divide(A=@n17.G, B=@n7) {desc=v in 1/N})
  )
  TextureSample!(UVs=@n19, Tex=@n15, Level=Constant() {r=-5}, Apply View MipBias) {constMipValue=1, mipValueMode=TMVM_MipLevel}
  TextureSample(UVs=@n19, Tex=@n15) {automaticViewMipBias=false, mipValueMode=TMVM_MipLevel}
  Time()
  WorldPosition()
 ]
}
