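# Masked lit material for Niagara mesh particles in the Grapefruit emote that makes a mesh appear with a textured wipe.
# Niagara-fed System Age, remapped up to Wipe Speed and square-rooted, offsets a mask texture projected on local XZ position to drive the reveal.
# A tint color, optionally multiplied by a texture whose alpha also cuts the mask, feeds diffuse and emissive terms.
# Panning shield caustics add bluish stripes to the emissive.
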Material /BRCosmetics/Effects/Fort_Effects/Emotes/Grapefruit/M_Wipeon
graph {
 properties {bUsedWithNiagaraMeshParticles=true, blendMode=BLEND_Masked, pixelDepthOffsetMode=PDOM_Legacy, usedShadingModels=Default Lit}
 nodes [
  n1 = Multiply(
   A=MaterialFunctionCall() {materialFunction=&LocalPosition}.Instance Local Position
   B=ScalarParameter() {defaultValue=0.05, parameterName=Wipe Freq, sliderMax=1}
  )
  ComponentMask(None=@n1) {b=true, g=true}
  Constant3Vector() {constant=<rgba 0 0.02 0.25 0>}
  n2 = ScalarParameter() {defaultValue=0.4, parameterName=Emissive Scale, sliderMax=1}
  n3 = TextureSample(UVs=TextureCoordinate(), Tex=TextureObjectParameter() {parameterName=Texture})
  n4 = VectorParameter() {parameterName=Color, defaultValue=<rgba 0 0.02 0.25 0>}
  n5 = ScalarParameter() {parameterName=Use Texture, sliderMax=1}
  n6 = LinearInterpolate(A=@n4.RGB, B=Multiply(A=@n3.RGB, B=@n4.RGB), Alpha=@n5) {constAlpha=0}
  n7 = ComponentMask(None=@n1) {b=true, r=true}
  n8 = Multiply(A=@n7) {constB=0.3}
  n9 = Time()
  n10 = Multiply(
   A=TextureSample(UVs=Panner(Coordinate=@n8, Time=@n9) {speedX=0.2}) {texture=&T_ShieldCaustics}.RGB
   B=Constant3Vector() {constant=<rgba 0.0253 0.1294 0.25 0>}
  )
  n11 = ScalarParameter() {defaultValue=0.1, parameterName=stripeBrightness, sliderMax=1}
  MaterialFunctionCall(Emissive=Add(A=Multiply(A=@n6, B=@n2), B=Multiply(A=@n10, B=@n11) {constB=0.03})) {materialFunction=&MF_FX_EmissiveAdjustments}
  Multiply(A=@n6, B=ScalarParameter() {defaultValue=1, parameterName=Diffuse Scale, sliderMax=1})
  n12 = TextureObjectParameter() {parameterName=WipeOn Texture, samplerType=SAMPLERTYPE_Masks, texture=/Game/Environments/Apollo/Sets/CubeTownV2/Materials/Parent/CubeTown_1_H}
  n13 = ScalarParameter() {defaultValue=0.3, parameterName=Wipe Speed, sliderMax=15}
  n14 = DynamicParameter() {parameterIndex=3, paramNames [System Age, "", "", ""]}
  n15 = Add(
   A=Subtract(A=TextureSample(UVs=@n7, Tex=@n12).RGB, B=Constant3Vector() {constant=<rgba 1 1 1 0>})
   B=Power(Base=InverseLinearInterpolate(B=@n13, Value=@n14.System Age) {bClampResult=true, constA=0.1}) {constExponent=0.5}
  )
  Saturate(None=LinearInterpolate(A=@n15, B=Multiply(A=@n15, B=@n3.A), Alpha=@n5) {constAlpha=0})
 ]
}
