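# Niagara system for the Grapefruit emote that stages an animated Pac-Man scene with maze, dots, four ghosts, Pac-Man and digits as mesh particles.
# Each emitter bursts particles that follow animated skeletal meshes passed as user parameters, so the animation drives their motion.
# Scratch modules pick flipbook frames from index arrays, including ghost eye and alternate states, and mask them using skeletal scale.
# Dynamic material parameters feed frame, state and intro time to the materials.
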NiagaraSystem /BRCosmetics/Effects/Fort_Effects/Emotes/Grapefruit/NS_Grapefruit_Emote_C
dependencies {
 usedDataInterfaces [&NiagaraDataInterfaceArrayInt32, &NiagaraDataInterfaceCurve]
 usedDynamicInputs []
 usedModules [&SystemState]
}
emitters {
 Mesh_Maze_PX {
  emitterUpdateScript [
   &EmitterState()
   &SpawnBurst_Instantaneous(Use Spawn Probability=false, UseLoopCountLimit=false)
  ]
  particleSpawnScript [
   &InitializeParticle(
    Color=<rgba 0 0.1 1 1>
    Lifetime=16.8
    Mesh Scale Mode=::NewEnumerator3
    Mesh Uniform Scale=$User.PixelSize
    Position Mode=::NewEnumerator0
    Sprite Rotation Mode=::NewEnumerator0
    Sprite Size Mode=::NewEnumerator3
    Uniform Sprite Size=&Multiply_Float(A=$User.PixelSize, B=16)
    Write Material Random=true
   )
   &SkeletalMeshLocation(
    Bone Sampling Mode=::NewEnumerator4
    MaskSpawnGroup=false
    Sampled Position Offset=&Add_Vector(A=<0 -15 135>, B=$User.Offset)
    Sampled Position Offset Coordinate Space:ENiagaraSkelMeshTransforms=::NewEnumerator1
    Skeletal Mesh=$User.SkelMesh_Root
    Use Sampled Mesh Attribute as Ribbon ID=false
   ) {module=SkeletalMeshLocation002}
  ]
  particleUpdateScript [
   &ParticleState()
   &DynamicMaterialParameters(
    Index 3 Param 1=$User.IntroTime
    Module.Param0WriteEnabled3=true
    Module.Param1WriteEnabled3=false
    Module.Param2WriteEnabled3=false
    Module.Param3WriteEnabled3=false
    Write Parameter Index 0=false
    Write Parameter Index 3=true
   )
   &CameraOffset(Camera Offset Amount=15)
  ]
  rendererClasses [&NiagaraMeshRendererProperties]
  renderers [
   {rendererClass=&NiagaraMeshRendererProperties, rendererIndex=0}
   {rendererClass=&NiagaraMeshRendererProperties, rendererIndex=1}
  ]
 }
 Mesh_Skel_Blinky {
  emitterSpawnScript [
   /SetVariables(Emitter.SubImageIndices=DataInterface() {propertyValues {GpuSyncMode=SyncCpuToGpu, MaxElements=0, IntData [14,15,14,15]}}) {isSetParameters=true, module=SetVariables_#1}
   /SetVariables(
    Emitter.SubImageIndices_Eyes=DataInterface() {propertyValues {GpuSyncMode=SyncCpuToGpu, MaxElements=0, IntData [47,47,47,47]}}
    Emitter.SubImageIndices_PP=DataInterface() {propertyValues {GpuSyncMode=SyncCpuToGpu, MaxElements=0, IntData [40,41,42,43]}}
   ) {isSetParameters=true}
  ]
  emitterUpdateScript [
   &EmitterState()
   &SpawnBurst_Instantaneous(Use Spawn Probability=false, UseLoopCountLimit=false)
  ]
  particleSpawnScript [&InitializeParticle(
    Color=<rgba 1 0 0 1>
    Lifetime=16.7
    Mesh Scale Mode=::NewEnumerator3
    Mesh Uniform Scale=$User.PixelSize
    Position Mode=::NewEnumerator0
    Sprite Rotation Mode=::NewEnumerator0
    Sprite Size Mode=::NewEnumerator3
    Uniform Sprite Size=&Multiply_Float(A=$User.PixelSize, B=16)
    Write Material Random=true
   )]
  particleUpdateScript [
   &ParticleState()
   &self:SubImageIndices_Ghosts_Scale(
    Curve=DataInterface() {propertyValues {Curve {postInfinityExtrap=RCCE_Cycle, preInfinityExtrap=RCCE_Cycle, keys [<at 0 0>,<at 1 1>]}}}
    SkeletalMesh:NiagaraDataInterfaceSkeletalMesh=$User.SkelMesh_Blinky
    Speed:NiagaraFloat=$User.PlaybackSpeed
    SubImageIndices:NiagaraDataInterfaceArrayInt32=$Emitter.SubImageIndices
    SubImageIndices_Eyes:NiagaraDataInterfaceArrayInt32=$Emitter.SubImageIndices_Eyes
    SubImageIndices_PP:NiagaraDataInterfaceArrayInt32=$Emitter.SubImageIndices_PP
   ) {module=SubImageIndices_Ghosts_Scale}
   &self:Mask_Subimage_Index(
    Mask_IsHidden:NiagaraFloat=$Particles.Mask_IsHidden
    Masked SubImageIndex:NiagaraFloat=$User.Masked_SubImageIndex
    SubImageIndex:NiagaraFloat=$Particles.SubImageIndex
    Threshold=0.95
   ) {module=Mask_Subimage_Index}
   &DynamicMaterialParameters(
    Index 0 Param 1=$Particles.SubImageIndex
    Index 0 Param 4=8
    Index 1 Param 1=$Particles.State
    Index 3 Param 1=$User.IntroTime
    Module.Param0WriteEnabled=true
    Module.Param0WriteEnabled1=true
    Module.Param0WriteEnabled3=true
    Module.Param1WriteEnabled=false
    Module.Param1WriteEnabled1=false
    Module.Param1WriteEnabled3=false
    Module.Param2WriteEnabled=false
    Module.Param2WriteEnabled1=false
    Module.Param2WriteEnabled3=false
    Module.Param3WriteEnabled=true
    Module.Param3WriteEnabled1=false
    Module.Param3WriteEnabled3=false
    Write Parameter Index 1=true
    Write Parameter Index 3=true
   )
   &SkeletalMeshLocation(
    Bone Sampling Mode=::NewEnumerator4
    "Interpolate Mesh Sampling (Update)"=true
    MaskSpawnGroup=false
    Sampled Position Offset=$User.Offset
    Skeletal Mesh=$User.SkelMesh_Blinky
    Use Sampled Mesh Attribute as Ribbon ID=false
   )
   /SetVariables(Particles.Velocity=$Particles.SkeletalMeshLocation.SampledVelocity) {isSetParameters=true}
   &CameraOffset(Camera Offset Amount=15)
  ]
  rendererClasses [&NiagaraMeshRendererProperties]
  renderers [{rendererClass=&NiagaraMeshRendererProperties, rendererIndex=0}]
 }
 Mesh_Skel_Clyde {
  emitterSpawnScript [
   /SetVariables(Emitter.SubImageIndices=DataInterface() {propertyValues {GpuSyncMode=SyncCpuToGpu, MaxElements=0, IntData [30,31,30,31]}}) {enabled=false, isSetParameters=true}
   /SetVariables(Emitter.SubImageIndices=DataInterface() {propertyValues {GpuSyncMode=SyncCpuToGpu, MaxElements=0, IntData [38,39,38,39]}}) {isSetParameters=true}
   /SetVariables(
    Emitter.SubImageIndices_Eyes=DataInterface() {propertyValues {GpuSyncMode=SyncCpuToGpu, MaxElements=0, IntData [47,47,47,47]}}
    Emitter.SubImageIndices_PP=DataInterface() {propertyValues {GpuSyncMode=SyncCpuToGpu, MaxElements=0, IntData [40,41,42,43]}}
   ) {isSetParameters=true, module=SetVariables_#2}
  ]
  emitterUpdateScript [
   &EmitterState()
   &SpawnBurst_Instantaneous(Use Spawn Probability=false, UseLoopCountLimit=false)
  ]
  particleSpawnScript [
   &InitializeParticle(
    Color=<rgba 1 0.6588 0 1>
    Lifetime=16.7
    Mesh Scale Mode=::NewEnumerator3
    Mesh Uniform Scale=$User.PixelSize
    Position Mode=::NewEnumerator0
    Sprite Rotation Mode=::NewEnumerator0
    Sprite Size Mode=::NewEnumerator3
    Uniform Sprite Size=&Multiply_Float(A=$User.PixelSize, B=16)
    Write Material Random=true
   ) {module=InitializeParticle001}
   &InitializeParticle(
    Color=<rgba 0 0.8039 1 1>
    Lifetime=16.7
    Mesh Scale Mode=::NewEnumerator3
    Mesh Uniform Scale=$User.PixelSize
    Position Mode=::NewEnumerator0
    Sprite Rotation Mode=::NewEnumerator0
    Sprite Size Mode=::NewEnumerator3
    Uniform Sprite Size=&Multiply_Float(A=$User.PixelSize, B=16)
    Write Material Random=true
   ) {enabled=false, module=InitializeParticle002}
  ]
  particleUpdateScript [
   &ParticleState()
   &self:SubImageIndices_Ghosts_Scale(
    Curve=DataInterface() {propertyValues {Curve {postInfinityExtrap=RCCE_Cycle, preInfinityExtrap=RCCE_Cycle, keys [<at 0 0>,<at 1 1>]}}}
    SkeletalMesh:NiagaraDataInterfaceSkeletalMesh=$User.SkelMesh_Clyde
    Speed:NiagaraFloat=$User.PlaybackSpeed
    SubImageIndices:NiagaraDataInterfaceArrayInt32=$Emitter.SubImageIndices
    SubImageIndices_Eyes:NiagaraDataInterfaceArrayInt32=$Emitter.SubImageIndices_Eyes
    SubImageIndices_PP:NiagaraDataInterfaceArrayInt32=$Emitter.SubImageIndices_PP
   ) {module=SubImageIndices_Ghosts_Scale}
   &self:Mask_Subimage_Index(
    Mask_IsHidden:NiagaraFloat=$Particles.Mask_IsHidden
    Masked SubImageIndex:NiagaraFloat=$User.Masked_SubImageIndex
    SubImageIndex:NiagaraFloat=$Particles.SubImageIndex
    Threshold=0.95
   ) {module=Mask_Subimage_Index}
   &DynamicMaterialParameters(
    Index 0 Param 1=$Particles.SubImageIndex
    Index 0 Param 4=8
    Index 1 Param 1=$Particles.State
    Index 3 Param 1=$User.IntroTime
    Module.Param0WriteEnabled=true
    Module.Param0WriteEnabled1=true
    Module.Param0WriteEnabled3=true
    Module.Param1WriteEnabled=false
    Module.Param1WriteEnabled1=false
    Module.Param1WriteEnabled3=false
    Module.Param2WriteEnabled=false
    Module.Param2WriteEnabled1=false
    Module.Param2WriteEnabled3=false
    Module.Param3WriteEnabled=true
    Module.Param3WriteEnabled1=false
    Module.Param3WriteEnabled3=false
    Write Parameter Index 1=true
    Write Parameter Index 3=true
   )
   &SkeletalMeshLocation(
    Bone Sampling Mode=::NewEnumerator4
    "Interpolate Mesh Sampling (Update)"=true
    MaskSpawnGroup=false
    Sampled Position Offset=$User.Offset
    Skeletal Mesh=$User.SkelMesh_Clyde
    Use Sampled Mesh Attribute as Ribbon ID=false
   )
   /SetVariables(Particles.Velocity=$Particles.SkeletalMeshLocation.SampledVelocity) {isSetParameters=true}
   &CameraOffset(Camera Offset Amount=15)
  ]
  rendererClasses [&NiagaraMeshRendererProperties]
  renderers [{rendererClass=&NiagaraMeshRendererProperties, rendererIndex=0}]
 }
 Mesh_Skel_Digits {
  emitterSpawnScript [/SetVariables(Emitter.SubImageIndices=DataInterface() {propertyValues {GpuSyncMode=SyncCpuToGpu, MaxElements=0, IntData [0,1,2,3]}}) {isSetParameters=true, module=SetVariables_#1}]
  emitterUpdateScript [
   &EmitterState()
   &SpawnBurst_Instantaneous(Spawn Count=4, Use Spawn Probability=false, UseLoopCountLimit=false)
  ]
  particleSpawnScript [&InitializeParticle(
    Color=<rgba 1 1 0 1>
    Lifetime=16.7
    Mesh Scale Mode=::NewEnumerator3
    Mesh Uniform Scale=$User.PixelSize
    Position Mode=::NewEnumerator0
    Sprite Rotation Mode=::NewEnumerator0
    Sprite Size Mode=::NewEnumerator3
    Uniform Sprite Size=&Multiply_Float(A=$User.PixelSize, B=16)
    Write Material Random=true
   )]
  particleUpdateScript [
   &ParticleState()
   &self:ScratchModule(
    Offset:Vector3f=&Add_Vector(A=<0 0 -15>, B=$User.Offset)
    SkelMesh:NiagaraDataInterfaceSkeletalMesh=$User.SkelMesh_Digits_C
   ) {module=ScratchModule}
   &DynamicMaterialParameters(
    Index 0 Param 1=$Particles.SubImageIndex
    Index 0 Param 4=8
    Index 1 Param 1=$Particles.SubImageIndex
    Index 1 Param 4=2
    Module.Param0WriteEnabled=true
    Module.Param0WriteEnabled1=true
    Module.Param1WriteEnabled=false
    Module.Param1WriteEnabled1=false
    Module.Param2WriteEnabled=false
    Module.Param2WriteEnabled1=false
    Module.Param3WriteEnabled=true
    Module.Param3WriteEnabled1=true
    Write Parameter Index 1=true
   )
   &Color(Color=/Niagara/DynamicInputs/LinearColor/MakeLinearColorFromFloat4(Vector:Vector4f=&MakeFloat4FromVectorAndFloat("Vector (RGB)"=<0 0.1 1>, "Float (Alpha)"=$Particles.Digits_Alpha)))
   &CameraOffset(Camera Offset Amount=15)
  ]
  rendererClasses [&NiagaraMeshRendererProperties]
  renderers [{rendererClass=&NiagaraMeshRendererProperties, rendererIndex=0}]
 }
 Mesh_Skel_Inky {
  emitterSpawnScript [
   /SetVariables(Emitter.SubImageIndices=DataInterface() {propertyValues {GpuSyncMode=SyncCpuToGpu, MaxElements=0, IntData [22,23,22,23]}}) {enabled=false, isSetParameters=true}
   /SetVariables(Emitter.SubImageIndices=DataInterface() {propertyValues {GpuSyncMode=SyncCpuToGpu, MaxElements=0, IntData [30,31,30,31]}}) {isSetParameters=true}
   /SetVariables(
    Emitter.SubImageIndices_Eyes=DataInterface() {propertyValues {GpuSyncMode=SyncCpuToGpu, MaxElements=0, IntData [47,47,47,47]}}
    Emitter.SubImageIndices_PP=DataInterface() {propertyValues {GpuSyncMode=SyncCpuToGpu, MaxElements=0, IntData [40,41,42,43]}}
   ) {isSetParameters=true, module=SetVariables_#2}
  ]
  emitterUpdateScript [
   &EmitterState()
   &SpawnBurst_Instantaneous(Use Spawn Probability=false, UseLoopCountLimit=false)
  ]
  particleSpawnScript [
   &InitializeParticle(
    Color=<rgba 0 0.8039 1 1>
    Lifetime=16.7
    Mesh Scale Mode=::NewEnumerator3
    Mesh Uniform Scale=$User.PixelSize
    Position Mode=::NewEnumerator0
    Sprite Rotation Mode=::NewEnumerator0
    Sprite Size Mode=::NewEnumerator3
    Uniform Sprite Size=&Multiply_Float(A=$User.PixelSize, B=16)
    Write Material Random=true
   ) {module=InitializeParticle001}
   &InitializeParticle(
    Color=<rgba 1 0.6588 0.8706 1>
    Lifetime=16.7
    Mesh Scale Mode=::NewEnumerator3
    Mesh Uniform Scale=$User.PixelSize
    Position Mode=::NewEnumerator0
    Sprite Rotation Mode=::NewEnumerator0
    Sprite Size Mode=::NewEnumerator3
    Uniform Sprite Size=&Multiply_Float(A=$User.PixelSize, B=16)
    Write Material Random=true
   ) {enabled=false}
  ]
  particleUpdateScript [
   &ParticleState()
   &self:SubImageIndices_Ghosts_Scale(
    Curve=DataInterface() {propertyValues {Curve {postInfinityExtrap=RCCE_Cycle, preInfinityExtrap=RCCE_Cycle, keys [<at 0 0>,<at 1 1>]}}}
    SkeletalMesh:NiagaraDataInterfaceSkeletalMesh=$User.SkelMesh_Inky
    Speed:NiagaraFloat=$User.PlaybackSpeed
    SubImageIndices:NiagaraDataInterfaceArrayInt32=$Emitter.SubImageIndices
    SubImageIndices_Eyes:NiagaraDataInterfaceArrayInt32=$Emitter.SubImageIndices_Eyes
    SubImageIndices_PP:NiagaraDataInterfaceArrayInt32=$Emitter.SubImageIndices_PP
   ) {module=SubImageIndices_Ghosts_Scale}
   &self:Mask_Subimage_Index(
    Mask_IsHidden:NiagaraFloat=$Particles.Mask_IsHidden
    Masked SubImageIndex:NiagaraFloat=$User.Masked_SubImageIndex
    SubImageIndex:NiagaraFloat=$Particles.SubImageIndex
    Threshold=0.95
   ) {module=Mask_Subimage_Index}
   &DynamicMaterialParameters(
    Index 0 Param 1=$Particles.SubImageIndex
    Index 0 Param 4=8
    Index 1 Param 1=$Particles.State
    Index 3 Param 1=$User.IntroTime
    Module.Param0WriteEnabled=true
    Module.Param0WriteEnabled1=true
    Module.Param0WriteEnabled3=true
    Module.Param1WriteEnabled=false
    Module.Param1WriteEnabled1=false
    Module.Param1WriteEnabled3=false
    Module.Param2WriteEnabled=false
    Module.Param2WriteEnabled1=false
    Module.Param2WriteEnabled3=false
    Module.Param3WriteEnabled=true
    Module.Param3WriteEnabled1=false
    Module.Param3WriteEnabled3=false
    Write Parameter Index 1=true
    Write Parameter Index 3=true
   )
   &SkeletalMeshLocation(
    Bone Sampling Mode=::NewEnumerator4
    "Interpolate Mesh Sampling (Update)"=true
    MaskSpawnGroup=false
    Sampled Position Offset=$User.Offset
    Skeletal Mesh=$User.SkelMesh_Inky
    Use Sampled Mesh Attribute as Ribbon ID=false
   )
   /SetVariables(Particles.Velocity=$Particles.SkeletalMeshLocation.SampledVelocity) {isSetParameters=true}
   &CameraOffset(Camera Offset Amount=15)
  ]
  rendererClasses [&NiagaraMeshRendererProperties]
  renderers [{rendererClass=&NiagaraMeshRendererProperties, rendererIndex=0}]
 }
 Mesh_Skel_PacMan {
  emitterSpawnScript [/SetVariables(Emitter.SubImageIndices=DataInterface() {propertyValues {GpuSyncMode=SyncCpuToGpu, MaxElements=0, IntData [0,1,2,3]}}) {isSetParameters=true, module=SetVariables_#1}]
  emitterUpdateScript [
   &EmitterState()
   &SpawnBurst_Instantaneous(Use Spawn Probability=false, UseLoopCountLimit=false)
  ]
  particleSpawnScript [&InitializeParticle(
    Color=<rgba 1 1 0 1>
    Lifetime=16.7
    Mesh Scale Mode=::NewEnumerator3
    Mesh Uniform Scale=$User.PixelSize
    Position Mode=::NewEnumerator0
    Sprite Rotation Mode=::NewEnumerator0
    Sprite Size Mode=::NewEnumerator3
    Uniform Sprite Size=&Multiply_Float(A=$User.PixelSize, B=16)
    Write Material Random=true
   )]
  particleUpdateScript [
   &ParticleState()
   &self:SubImageIndex_PacMan(
    Curve=DataInterface() {propertyValues {Curve {
       postInfinityExtrap=RCCE_Cycle
       preInfinityExtrap=RCCE_Cycle
       keys [{tangentMode=RCTM_User, time=0, value=0}, {tangentMode=RCTM_User, time=1, value=1}]
      }}}
    Speed:NiagaraFloat=$User.PlaybackSpeed
    SubImageIndices:NiagaraDataInterfaceArrayInt32=$Emitter.SubImageIndices
   ) {module=SubImageIndex_PacMan}
   &self:Mask_SubImageIndex_By_Scale_Pacman(
    Masked SubImageIndex:NiagaraFloat=$User.Masked_SubImageIndex
    SkeletalMesh:NiagaraDataInterfaceSkeletalMesh=$User.SkelMesh_PacMan
    SubImageIndex:NiagaraFloat=$Particles.SubImageIndex
    Threshold=0.9
   ) {module=Mask_SubImageIndex_By_Scale_Pacman}
   &SkeletalMeshLocation(
    Bone Sampling Mode=::NewEnumerator4
    "Interpolate Mesh Sampling (Update)"=true
    MaskSpawnGroup=false
    Sampled Position Offset=$User.Offset
    Skeletal Mesh=$User.SkelMesh_PacMan
    Use Sampled Mesh Attribute as Ribbon ID=false
   )
   &Timeline(Instantaneous Rewind=true, Timeline Beginning=4.8333, Timeline End=16.7)
   &self:MaskRoundPacman(Age Cutoff=6.32, Index=3, IntroTime:NiagaraFloat=$User.IntroTime) {module=MaskRoundPacman}
   &DynamicMaterialParameters(
    Index 0 Param 1=$Particles.SubImageIndex
    Index 0 Param 4=8
    Index 1 Param 4=2
    Index 3 Param 1=$User.IntroTime
    Module.Param0WriteEnabled=true
    Module.Param0WriteEnabled1=false
    Module.Param0WriteEnabled3=true
    Module.Param1WriteEnabled=false
    Module.Param1WriteEnabled1=false
    Module.Param1WriteEnabled3=false
    Module.Param2WriteEnabled=false
    Module.Param2WriteEnabled1=false
    Module.Param2WriteEnabled3=false
    Module.Param3WriteEnabled=true
    Module.Param3WriteEnabled1=true
    Module.Param3WriteEnabled3=false
    Write Parameter Index 1=true
    Write Parameter Index 3=true
   )
   /SetVariables(Particles.Velocity=$Particles.SkeletalMeshLocation.SampledVelocity) {isSetParameters=true}
   &CameraOffset(Camera Offset Amount=15)
  ]
  rendererClasses [&NiagaraMeshRendererProperties]
  renderers [{rendererClass=&NiagaraMeshRendererProperties, rendererIndex=0}]
 }
 Mesh_Skel_Pink {
  emitterSpawnScript [
   /SetVariables(Emitter.SubImageIndices=DataInterface() {propertyValues {GpuSyncMode=SyncCpuToGpu, MaxElements=0, IntData [38,39,38,39]}}) {enabled=false, isSetParameters=true}
   /SetVariables(Emitter.SubImageIndices=DataInterface() {propertyValues {GpuSyncMode=SyncCpuToGpu, MaxElements=0, IntData [22,23,22,23]}}) {isSetParameters=true}
   /SetVariables(
    Emitter.SubImageIndices_Eyes=DataInterface() {propertyValues {GpuSyncMode=SyncCpuToGpu, MaxElements=0, IntData [47,47,47,47]}}
    Emitter.SubImageIndices_PP=DataInterface() {propertyValues {GpuSyncMode=SyncCpuToGpu, MaxElements=0, IntData [40,41,42,43]}}
   ) {isSetParameters=true, module=SetVariables_#2}
  ]
  emitterUpdateScript [
   &EmitterState()
   &SpawnBurst_Instantaneous(Use Spawn Probability=false, UseLoopCountLimit=false)
  ]
  particleSpawnScript [
   &InitializeParticle(
    Color=<rgba 1 0.6588 0.8706 1>
    Lifetime=16.7
    Mesh Scale Mode=::NewEnumerator3
    Mesh Uniform Scale=$User.PixelSize
    Position Mode=::NewEnumerator0
    Sprite Rotation Mode=::NewEnumerator0
    Sprite Size Mode=::NewEnumerator3
    Uniform Sprite Size=&Multiply_Float(A=$User.PixelSize, B=16)
    Write Material Random=true
   ) {module=InitializeParticle001}
   &InitializeParticle(
    Color=<rgba 1 0.6588 0 1>
    Lifetime=16.7
    Mesh Scale Mode=::NewEnumerator3
    Mesh Uniform Scale=$User.PixelSize
    Position Mode=::NewEnumerator0
    Sprite Rotation Mode=::NewEnumerator0
    Sprite Size Mode=::NewEnumerator3
    Uniform Sprite Size=&Multiply_Float(A=$User.PixelSize, B=16)
    Write Material Random=true
   ) {enabled=false}
  ]
  particleUpdateScript [
   &ParticleState()
   &self:SubImageIndices_Ghosts_Scale(
    Curve=DataInterface() {propertyValues {Curve {postInfinityExtrap=RCCE_Cycle, preInfinityExtrap=RCCE_Cycle, keys [<at 0 0>,<at 1 1>]}}}
    SkeletalMesh:NiagaraDataInterfaceSkeletalMesh=$User.SkelMesh_Pinky
    Speed:NiagaraFloat=$User.PlaybackSpeed
    SubImageIndices:NiagaraDataInterfaceArrayInt32=$Emitter.SubImageIndices
    SubImageIndices_Eyes:NiagaraDataInterfaceArrayInt32=$Emitter.SubImageIndices_Eyes
    SubImageIndices_PP:NiagaraDataInterfaceArrayInt32=$Emitter.SubImageIndices_PP
   ) {module=SubImageIndices_Ghosts_Scale}
   &self:Mask_Subimage_Index(
    Mask_IsHidden:NiagaraFloat=$Particles.Mask_IsHidden
    Masked SubImageIndex:NiagaraFloat=$User.Masked_SubImageIndex
    SubImageIndex:NiagaraFloat=$Particles.SubImageIndex
    Threshold=0.95
   ) {module=Mask_Subimage_Index}
   &DynamicMaterialParameters(
    Index 0 Param 1=$Particles.SubImageIndex
    Index 0 Param 4=8
    Index 1 Param 1=$Particles.State
    Index 3 Param 1=$User.IntroTime
    Module.Param0WriteEnabled=true
    Module.Param0WriteEnabled1=true
    Module.Param0WriteEnabled3=true
    Module.Param1WriteEnabled=false
    Module.Param1WriteEnabled1=false
    Module.Param1WriteEnabled3=false
    Module.Param2WriteEnabled=false
    Module.Param2WriteEnabled1=false
    Module.Param2WriteEnabled3=false
    Module.Param3WriteEnabled=true
    Module.Param3WriteEnabled1=false
    Module.Param3WriteEnabled3=false
    Write Parameter Index 1=true
    Write Parameter Index 3=true
   )
   &SkeletalMeshLocation(
    Bone Sampling Mode=::NewEnumerator4
    "Interpolate Mesh Sampling (Update)"=true
    MaskSpawnGroup=false
    Sampled Position Offset=$User.Offset
    Skeletal Mesh=$User.SkelMesh_Pinky
    Use Sampled Mesh Attribute as Ribbon ID=false
   )
   /SetVariables(Particles.Velocity=$Particles.SkeletalMeshLocation.SampledVelocity) {isSetParameters=true}
   &CameraOffset(Camera Offset Amount=15)
  ]
  rendererClasses [&NiagaraMeshRendererProperties]
  renderers [{rendererClass=&NiagaraMeshRendererProperties, rendererIndex=0}]
 }
 Test_Mesh_Dots_SkelScale {
  emitterUpdateScript [
   &EmitterState()
   &SpawnBurst_Instantaneous(Spawn Count=43, Use Spawn Probability=false, UseLoopCountLimit=false)
  ]
  particleSpawnScript [&InitializeParticle(
    Color=<rgba 0.474 0.27 0.1753 0>
    Lifetime=16.7
    Mesh Scale Mode=::NewEnumerator3
    Mesh Uniform Scale=&MakeCustomFloatFromBool(
     Bool=&SetBoolByIntComparison(A=&ReturnExecIndex(), B=30, Comparison Type=::NewEnumerator4)
     True Float=2
     False Float=1
    )
    Position Mode=::NewEnumerator0
    Sprite Rotation Mode=::NewEnumerator0
    Sprite Size Mode=::NewEnumerator3
    Uniform Sprite Size=&Multiply_Float(A=$User.PixelSize, B=16)
    Write Material Random=true
   ) {module=InitializeParticle004}]
  particleUpdateScript [
   &ParticleState()
   /SetVariables(Particles.Position=&self:SetPosition(New Skeletal Mesh:NiagaraDataInterfaceSkeletalMesh=$User.SkelMesh_Dots, Offset:Vector3f=$User.Offset)) {isSetParameters=true}
   &DynamicMaterialParameters(Write Parameter Index 0=false)
   &self:Mask_SkelMesh_Scaling_Dots(NumBones=43, SkeletalMesh:NiagaraDataInterfaceSkeletalMesh=$User.SkelMesh_Dots) {module=Mask_SkelMesh_Scaling_Dots}
   &Color(Color=&MakeLinearColorFromFloat4(Vector:Vector4f=&MakeFloat4FromVectorAndFloat("Vector (RGB)"=<0.474 0.27 0.1753>, "Float (Alpha)"=$Particles.Mask_Scale))) {module=Color002}
   &CameraOffset(Camera Offset Amount=15) {module=CameraOffset001}
  ]
  rendererClasses [&NiagaraMeshRendererProperties]
  renderers [{rendererClass=&NiagaraMeshRendererProperties, rendererIndex=0}]
 }
}
summary {
 systemName=NS_Grapefruit_Emote_C
 emitters [
  {emitterName=Mesh_Skel_PacMan, rendererClasses [&NiagaraMeshRendererProperties]}
  {emitterName=Mesh_Skel_Blinky, rendererClasses [&NiagaraMeshRendererProperties]}
  {emitterName=Mesh_Maze_PX, rendererClasses [&NiagaraMeshRendererProperties]}
  {emitterName=Test_Mesh_Dots_SkelScale, rendererClasses [&NiagaraMeshRendererProperties]}
  {emitterName=Mesh_Skel_Clyde, rendererClasses [&NiagaraMeshRendererProperties]}
  {emitterName=Mesh_Skel_Inky, rendererClasses [&NiagaraMeshRendererProperties]}
  {emitterName=Mesh_Skel_Digits, rendererClasses [&NiagaraMeshRendererProperties]}
  {emitterName=Mesh_Skel_Pink, rendererClasses [&NiagaraMeshRendererProperties]}
 ]
 userVariables [
  User.bIsFrontEndPreview {type=NiagaraBool}
  User.IntroTime {type=NiagaraFloat, defaultValue=0, description=""}
  User.IsClient {type=NiagaraBool}
  User.Masked_SubImageIndex {type=NiagaraFloat, defaultValue=50, description=""}
  User.NumPixelsX {type=NiagaraInt32, defaultValue=16}
  User.NumPixelsY {type=NiagaraInt32, defaultValue=16}
  User.Offset {type=Vector3f, description="", defaultValue {struct=Vector3f, value=<0 0 30>}}
  User.PixelSize {type=NiagaraFloat, defaultValue=0.75, description=""}
  User.PlaybackSpeed {type=NiagaraFloat, defaultValue=3, description=""}
  User.SkelMesh_Blinky {
   type=NiagaraDataInterfaceSkeletalMesh
   defaultValue {
    struct=NiagaraExt_VariableValue_DataInterface
    value {dataInterface=&self:NiagaraDataInterfaceSkeletalMesh_1}
   }
  }
  User.SkelMesh_Clyde {
   type=NiagaraDataInterfaceSkeletalMesh
   defaultValue {
    struct=NiagaraExt_VariableValue_DataInterface
    value {dataInterface=&self:NiagaraDataInterfaceSkeletalMesh_2}
   }
  }
  User.SkelMesh_Digits_C {
   type=NiagaraDataInterfaceSkeletalMesh
   defaultValue {
    struct=NiagaraExt_VariableValue_DataInterface
    value {dataInterface=&self:NiagaraDataInterfaceSkeletalMesh_10}
   }
  }
  User.SkelMesh_Dots {
   type=NiagaraDataInterfaceSkeletalMesh
   defaultValue {
    struct=NiagaraExt_VariableValue_DataInterface
    value {dataInterface=&self:NiagaraDataInterfaceSkeletalMesh_6}
   }
  }
  User.SkelMesh_Inky {
   type=NiagaraDataInterfaceSkeletalMesh
   defaultValue {
    struct=NiagaraExt_VariableValue_DataInterface
    value {dataInterface=&self:NiagaraDataInterfaceSkeletalMesh_3}
   }
  }
  User.SkelMesh_PacMan {
   type=NiagaraDataInterfaceSkeletalMesh
   defaultValue {
    struct=NiagaraExt_VariableValue_DataInterface
    value {dataInterface=&self:NiagaraDataInterfaceSkeletalMesh_0}
   }
  }
  User.SkelMesh_Pinky {
   type=NiagaraDataInterfaceSkeletalMesh
   defaultValue {
    struct=NiagaraExt_VariableValue_DataInterface
    value {dataInterface=&self:NiagaraDataInterfaceSkeletalMesh_4}
   }
  }
  User.SkelMesh_Root {
   type=NiagaraDataInterfaceSkeletalMesh
   defaultValue {
    struct=NiagaraExt_VariableValue_DataInterface
    value {dataInterface=&self:NiagaraDataInterfaceSkeletalMesh_5}
   }
  }
 ]
}
